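# Niagara skydiving contrail of tumbling bubble-tea cups that spill boba, ice and liquid, tinted by a color variant.
# Cups spawn on a ring and move along System.BobaVelocity, with separately animated lids and shrinking liquid.
# Boba/ice, liquid sprite and ribbon emitters spawn from cups via particle reads, inherit their color, and use position offsets to swirl downstream.
# Liquid_Trail is disabled; Liquid_Trail002 is the active ribbon, ordered by sampled cup age.
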NiagaraSystem /BRCosmetics/Athena/Effects/Particles/Skydiving/NS_Contrail_SharpFang/NS_Contrail_SharpFang
dependencies {
 usedDataInterfaces [&NiagaraDataInterfaceArrayColor, &NiagaraDataInterfaceCurve, &NiagaraDataInterfaceParticleRead]
 usedDynamicInputs [
  &TransformVector
  &MakeFloatFromVector
  &Lerp_Vector
  &SetBoolByFloatComparison
  &MakeIntFromFloat
 ]
 usedModules [
  &SystemState
  &self:SystemSpawnScript.SystemScriptSource.SystemScriptGraph.NiagaraNodeAssignment_2.SetVariables_#1
  &self:SystemSpawnScript.SystemScriptSource.SystemScriptGraph.NiagaraNodeAssignment_3.SetVariables_#2
  &self:SystemSpawnScript.SystemScriptSource.SystemScriptGraph.NiagaraNodeAssignment_1.SetVariables_#3
  &InterpolateOverTime
  &FN_Integrator
  &Timeline
  &self:SystemSpawnScript.SystemScriptSource.SystemScriptGraph.NiagaraNodeAssignment_4.SetVariables_#4
  &self:SystemSpawnScript.SystemScriptSource.SystemScriptGraph.NiagaraNodeAssignment_0.SetVariables_#5
 ]
}
emitters {
 Boba_and_Ice {
  emitterUpdateScript [
   &EmitterState()
   &SpawnParticlesFromOtherEmitter(
    Attribute Reader=DataInterface() {propertyValues {EmitterBinding {bindingMode=Other, emitterName=Cups}}}
    Minimum Age Spawn Threshold=0.5
    Spawn Rate=60
    Spawn Rate Per Particle Cap=200
    Use Maximum Age=false
    Use Minimum Age=false
   )
  ]
  particleSpawnScript [
   /SetVariables(Particles.VisibilityTag=&RandomRangeInt(Minimum=0, Maximum=1)) {isSetParameters=true}
   &InitializeParticle(
    Color Mode=::NewEnumerator0
    Lifetime Max=1.25
    Lifetime Min=0.75
    Lifetime Mode=::NewEnumerator1
    Mesh Scale Mode=::NewEnumerator1
    Mesh Uniform Scale Max=0.3
    Mesh Uniform Scale Min=0.2
    Sprite Rotation Mode=::NewEnumerator2
    Sprite Size Mode=::NewEnumerator1
    Uniform Sprite Size Max=4.5
    Uniform Sprite Size Min=4
    UsePositionOffset=false
    Write Material Random=true
   )
   &ShapeLocation(
    Apply To Particle Position=::NewEnumerator1
    MaskSpawnGroup=false
    Override Seed=false
    Shape Origin=&VectorFromFloat(Value=0)
    Sphere Radius=6
    UseOwnerScale=true
   )
   &SampleParticlesFromOtherEmitter(
    Applied Position Offset:Vector3f=$Particles.ShapeLocation.ShapePosition
    Color Sampling=::NewEnumerator1
    Maximum Sampled Particle Age=0.75
    Mesh Orientation Sampling:ENiagara_DirectReadSamplingMode=::NewEnumerator1
    Minimum Sampled Particle Age=0.4
    Module.MaxSampledAgeCutoff=true
    Module.MinSampledAgeCutoff=true
    Particle ID Sampling=::NewEnumerator1
    Reject Particle By Sampled Age=true
   )
   &InitialMeshOrientation(Use Rotation Vector=false)
   &Color(Color=&LerpLinearColors(
     StartColor=&MakeLinearColorFromVectorAndFloat(
      "Vector (RGB)"=&MakeVectorFromLinearColorRGB(Color=$Output.SampleParticlesFromOtherEmitter.SampledColor)
      "Float (Alpha)"=1
     )
     EndColor=<rgba 1 1 1 1>
     LerpFactor=&MakeFloatFromInt(Int=$Particles.VisibilityTag)
    ))
   /SetVariables(Particles.TrailDirection:Vector3f=&Add_Vector(
     A=$System.BobaVelocity
     B=&Multiply_VectorByFloat(
      Vector=&ReturnMeshOrientationAxis(Mesh Orientation Axis=::ZAxis, Mesh Orientation Quat=$Output.SampleParticlesFromOtherEmitter.SampledMeshOrientation)
      Float=$System.SwirlSpread
     )
    )) {enabled=false, isSetParameters=true}
   /SetVariables(Particles.TrailDirection:Vector3f=&Add_Vector(
     A=$System.BobaVelocity
     B=&MakeVector(
      X=&Sine(Normalized Angle=$Emitter.Age, Period=0.66, Scale=$System.SwirlSpread, Bias=0)
      Y=&Cosine(Normalized Angle=$Emitter.Age, Period=0.66, Scale=$System.SwirlSpread, Bias=0)
      Z=0
     )
    )) {enabled=false, isSetParameters=true}
   /SetVariables(Particles.TrailDirection:Vector3f=&MakeVector(
     X=&Sine(Normalized Angle=$Emitter.Age, Period=0.66, Scale=$System.SwirlSpread, Bias=0)
     Y=&Cosine(Normalized Angle=$Emitter.Age, Period=0.66, Scale=$System.SwirlSpread, Bias=0)
     Z=0
    )) {isSetParameters=true}
  ]
  particleUpdateScript [
   &ParticleState()
   /SetVariables(Particles.PositionOffset:NiagaraPosition=&Add_Vector(
     A=Linked() {linkedVariable {name=Particles.Position, type=NiagaraPosition}}
     B=&Multiply_VectorByFloat(
      Vector=&Add_Vector(A=$System.BobaVelocity, B=$Particles.TrailDirection)
      Float=&FN_FrontEndFloatModifier(
       Gameplay=&FloatFromCurve(
        FloatCurve=DataInterface() {propertyValues {Curve {keys [
            {arriveTangent=0.5537, interpMode=RCIM_Cubic, leaveTangent=0.5537, tangentMode=RCTM_User, time=0, value=0}
            {arriveTangent=1.1281, interpMode=RCIM_Cubic, leaveTangent=1.1281, tangentMode=RCTM_User, time=1, value=1}
           ]}}}
        Scale Curve=&Multiply_Float(A=$System.BobaDistance, B=$Particles.Lifetime)
       )
       Front End=&FloatFromCurve(
        FloatCurve=DataInterface() {propertyValues {Curve {keys [
            {arriveTangent=0.5537, interpMode=RCIM_Cubic, leaveTangent=0.5537, tangentMode=RCTM_User, time=0, value=0}
            {arriveTangent=1.1281, interpMode=RCIM_Cubic, leaveTangent=1.1281, tangentMode=RCTM_User, time=1, value=1}
           ]}}}
        Scale Curve=&Multiply_Float(A=$System.BobaDistance, B=$Particles.Lifetime)
       )
      )
     )
    )) {isSetParameters=true}
   &UpdateMeshOrientation(
    Rotation Rate=&RandomRangeFloat(Minimum=0.5, Maximum=1.5)
    Rotation Vector=&RandomVector(Vector Scale=1, Constrain to Hemisphere=true, Hemisphere Constraint Vector=<0 0 1>)
   )
   &ScaleMeshSize(Scale Factor=&VectorFromFloat(Value=&FloatFromCurve(FloatCurve=DataInterface() {propertyValues {Curve {keys [
          {arriveTangent=25.3602, interpMode=RCIM_Cubic, leaveTangent=25.3602, tangentMode=RCTM_User, time=0, value=0}
          {interpMode=RCIM_Cubic, tangentMode=RCTM_User, time=0.1, value=1}
          {arriveTangent=0.0772, interpMode=RCIM_Cubic, leaveTangent=0.0772, tangentMode=RCTM_User, time=0.8, value=1}
          {arriveTangent=-5.7608, interpMode=RCIM_Cubic, leaveTangent=-5.7608, tangentMode=RCTM_User, time=1, value=0}
         ]}}})))
   &ScaleSpriteSize(Uniform Curve Sprite Scale=DataInterface() {propertyValues {Curve {keys [
        {tangentMode=RCTM_User, time=0, value=1}
        {arriveTangent=-0.0405, interpMode=RCIM_Cubic, leaveTangent=-0.0405, tangentMode=RCTM_User, time=0.85, value=1}
        {arriveTangent=-8.0089, interpMode=RCIM_Cubic, leaveTangent=-8.0089, tangentMode=RCTM_User, time=1, value=0}
       ]}}}) {enabled=false}
   &ScaleColor(
    Scale Alpha=&FloatFromCurve(FloatCurve=DataInterface() {propertyValues {Curve {keys [
         {tangentMode=RCTM_User, time=0, value=1}
         {arriveTangent=-0.0405, interpMode=RCIM_Cubic, leaveTangent=-0.0405, tangentMode=RCTM_User, time=0.7009, value=1}
         {arriveTangent=-8.0089, interpMode=RCIM_Cubic, leaveTangent=-8.0089, tangentMode=RCTM_User, time=1, value=0}
        ]}}})
    ScaleA=true
    ScaleRGB=true
   )
  ]
  rendererClasses [&NiagaraMeshRendererProperties, &NiagaraSpriteRendererProperties]
  renderers [
   {rendererClass=&NiagaraMeshRendererProperties, rendererIndex=0}
   {rendererClass=&NiagaraSpriteRendererProperties, rendererIndex=1}
  ]
 }
 Cups {
  emitterUpdateScript [
   &EmitterState()
   &SpawnBurst_Instantaneous(Use Spawn Probability=false, UseLoopCountLimit=false) {enabled=false}
   &SpawnRate(SpawnRate=1.5, Use Spawn Probability=false)
  ]
  particleSpawnScript [
   &InitializeParticle(Lifetime=1.6666, Mesh Scale Mode=::NewEnumerator3, UsePositionOffset=false, Write Material Random=true)
   &InitialMeshOrientation(
    Mesh Axis To Orient=<0 0 1>
    Mesh Orientation Mode=::NewEnumerator2
    Orientation Vector=<0 0 1>
    Rotation=&RandomRangeVector(Minimum=<0 0 0>, Maximum=<0 0 1>)
    Use Rotation Vector=true
   ) {module=InitialMeshOrientation001}
   /SetVariables(Particles.InitialOrientation:NiagaraFloat=&RandomRangeFloat(Minimum=-1, Maximum=1)) {isSetParameters=true}
   &Color(Color=&SelectLinearColorFromArray(
     Array Sampling Mode=::NewEnumerator1
     Direct Array Index=&MakeCustomIntFromBool(
      Bool=&SetBoolByIntComparison(A=$System.ColorVariant, B=0, Comparison Type=::NewEnumerator4)
      True INT=&ReturnParticleIDAttributes(Particle ID Attribute=::NewEnumerator1)
      False INT=/Niagara/DynamicInputs/Subtract/Subtract_Integer(A:NiagaraInt32=$System.ColorVariant, B:NiagaraInt32=1)
     )
     Direct Array Mode=::NewEnumerator1
     Color Selection Array=DataInterface() {propertyValues {ColorData [<rgba 0.305 0.166 1 1>, <rgba 0.2004 1 0.166 1>, <rgba 1 0.5 0.25 1>]}}
     Random Seed=?
    ))
   /SetVariables() {isSetParameters=true}
   /SetVariables(Particles.CupID:NiagaraFloat=&MakeFloatFromInt(Int=&ReturnParticleIDAttributes())) {isSetParameters=true}
  ]
  particleUpdateScript [
   &ParticleState()
   &ShapeLocation(
    MaskSpawnGroup=false
    Offset=<0 25 75>
    Offset Coordinate Space=::Local
    Offset Mode=::NewEnumerator0
    Override Seed=false
    Ring / Disc Distribution Mode:ENiagaraSphereDistributionMode=::NewEnumerator1
    Ring Radius=75
    Rotation Angle:NiagaraFloat=&RandomRangeFloat(Minimum=80, Maximum=100)
    Rotation Angle Type:ENiagara_AngleInput=::NewEnumerator0
    Rotation Axis=<0 0 1>
    Rotation Mode=::NewEnumerator1
    Shape Primitive=::NewEnumerator4
    U Position=&Add_Float(
     A=&Multiply_Float(A=$Particles.Initial.CupID, B=0.6667)
     B=&FloatFromCurve(
      FloatCurve=DataInterface() {propertyValues {Curve {keys [{tangentMode=RCTM_User, time=0, value=-1}, {tangentMode=RCTM_User, time=1, value=1}]}}}
      Scale Curve=0.05
     )
    )
    UseOwnerScale=true
   )
   &ShapeLocation(
    MaskSpawnGroup=false
    Offset=<0 0 25>
    Offset Coordinate Space=::Local
    Offset Mode=::NewEnumerator0
    Override Seed=false
    Ring Radius=60
    Rotation Angle:NiagaraFloat=&RandomRangeFloat(Minimum=80, Maximum=100)
    Rotation Angle Type:ENiagara_AngleInput=::NewEnumerator0
    Rotation Axis=<0 0 1>
    Rotation Mode=::NewEnumerator1
    Shape Primitive=::NewEnumerator4
    UseOwnerScale=true
   ) {enabled=false, module=ShapeLocation001}
   &InheritSourceMovement(ApplyForce=false, ApplyPosition=true, ApplyPositionFalloff=false) {enabled=false}
   /SetVariables(Particles.MeshOrientation=&MultiplyQuaternion(
     Quaternion A=&FindQuatBetweenTwoVectors(
      From Vector=&NormalizeVector(
       Vector To Normalize=&MakeVector(
        X=&Sine(
         Normalized Angle=&Add_Float(A=$Emitter.Age, B=$Particles.Initial.InitialOrientation)
         Period=&MakeCustomFloatFromBool(Bool=&RandomBool(), True Float=1, False Float=-1)
         Scale=0.5
         Bias=0
        )
        Y=&Cosine(
         Normalized Angle=&Add_Float(A=$Emitter.Age, B=$Particles.Initial.InitialOrientation)
         Period=&MakeCustomFloatFromBool(Bool=&RandomBool(), True Float=1, False Float=-1)
         Scale=0.5
         Bias=0
        )
        Z=&FloatFromCurve(FloatCurve=DataInterface() {propertyValues {Curve {keys [{interpMode=RCIM_Cubic, tangentMode=RCTM_User, time=0.3, value=1}, {interpMode=RCIM_Cubic, tangentMode=RCTM_User, time=1, value=-1}]}}})
       )
       Safe Normalize=true
       Safe Fallback Threshold=0
       Safe Fallback Vector=<0 0 0.001>
      )
      To Vector=<0 0 1>
     )
     Quaternion B=$Particles.Initial.MeshOrientation
    )) {isSetParameters=true}
   /SetVariables(Particles.MeshOrientation_Lid:Quat4f=&LerpQuaternion(
     Quaternion A:Quat4f=$Particles.MeshOrientation
     Quaternion B:Quat4f=$Particles.Initial.MeshOrientation
     Lerp Factor:NiagaraFloat=&FloatFromCurve(
      FloatCurve=DataInterface() {propertyValues {Curve {keys [
          {interpMode=RCIM_Cubic, tangentMode=RCTM_User, time=0.33, value=0}
          {arriveTangent=1.5235, interpMode=RCIM_Cubic, leaveTangent=1.5235, tangentMode=RCTM_User, time=1, value=1}
         ]}}}
      Scale Curve=0.9
     )
     Enable Spherical Linear Interpolation:NiagaraBool=false
    )) {isSetParameters=true}
   /SetVariables(Particles.MeshOrientation=&MultiplyQuaternion(
     Quaternion A=&FindQuatBetweenTwoVectors(From Vector=<0 0 -1>, To Vector=$System.BobaVelocity)
     Quaternion B=$Particles.MeshOrientation
    )) {isSetParameters=true}
   /SetVariables(Particles.MeshOrientation_Lid:Quat4f=&MultiplyQuaternion(
     Quaternion A=&FindQuatBetweenTwoVectors(From Vector=<0 0 -1>, To Vector=$System.BobaVelocity)
     Quaternion B=$Particles.MeshOrientation_Lid
    )) {isSetParameters=true}
   /SetVariables(Particles.Position=&Add_Vector(
     A=$Particles.ShapeLocation.ShapePosition
     B=&Multiply_VectorByFloat(
      Vector=&Add_Vector(A=$System.BobaVelocity, B=<0 0 0>)
      Float=&FloatFromCurve(
       FloatCurve=DataInterface() {propertyValues {Curve {keys [
           {arriveTangent=0.4694, interpMode=RCIM_Cubic, leaveTangent=0.4694, tangentMode=RCTM_User, time=0, value=-0.4}
           {arriveTangent=0.9135, interpMode=RCIM_Cubic, leaveTangent=0.9135, tangentMode=RCTM_User, time=0.7, value=0.1}
           {arriveTangent=2.9973, interpMode=RCIM_Cubic, leaveTangent=2.9973, tangentMode=RCTM_User, time=1, value=0.7}
          ]}}}
       Scale Curve=$System.CupDistance
      )
     )
    )) {isSetParameters=true}
   /SetVariables(Particles.Position_Lid:NiagaraPosition=&Add_Vector(
     A=$Particles.ShapeLocation.ShapePosition
     B=&Multiply_VectorByFloat(
      Vector=&Add_Vector(A=$System.BobaVelocity, B=<0 0 0>)
      Float=&FloatFromCurve(
       FloatCurve=DataInterface() {propertyValues {Curve {keys [
           {arriveTangent=0.4694, interpMode=RCIM_Cubic, leaveTangent=0.4694, tangentMode=RCTM_User, time=0, value=-0.4}
           {arriveTangent=0.6515, interpMode=RCIM_Cubic, leaveTangent=0.6515, tangentMode=RCTM_Break, time=0.3, value=-0.218}
           {arriveTangent=2.4059, interpMode=RCIM_Cubic, leaveTangent=2.4059, tangentMode=RCTM_User, time=1, value=0.9}
          ]}}}
       Scale Curve=$System.CupDistance
      )
     )
    )) {isSetParameters=true}
   &ScaleMeshSize(Scale Factor=&VectorFromFloat(Value=&FloatFromCurve(FloatCurve=DataInterface() {propertyValues {Curve {keys [
          {arriveTangent=2.0977, interpMode=RCIM_Cubic, leaveTangent=2.0977, tangentMode=RCTM_User, time=0, value=0}
          {interpMode=RCIM_Cubic, tangentMode=RCTM_User, time=0.15, value=1}
          {arriveTangent=0.0772, interpMode=RCIM_Cubic, leaveTangent=0.0772, tangentMode=RCTM_User, time=0.9, value=1}
          {arriveTangent=-5.7608, interpMode=RCIM_Cubic, leaveTangent=-5.7608, tangentMode=RCTM_User, time=1, value=0}
         ]}}})))
   /SetVariables(Particles.ScaleLiquid:Vector3f=&Multiply_Vector(
     A=$Particles.Scale
     B=&MakeVector(
      X=&FloatFromCurve(
       FloatCurve=DataInterface() {propertyValues {Curve {keys [
           {interpMode=RCIM_Cubic, tangentMode=RCTM_User, time=0.6, value=1}
           {arriveTangent=-3.8261, interpMode=RCIM_Cubic, leaveTangent=-3.8261, tangentMode=RCTM_User, time=0.7, value=0}
          ]}}}
       Scale Curve=0.9
      )
      Y=&FloatFromCurve(
       FloatCurve=DataInterface() {propertyValues {Curve {keys [
           {interpMode=RCIM_Cubic, tangentMode=RCTM_User, time=0.6, value=1}
           {arriveTangent=-3.8261, interpMode=RCIM_Cubic, leaveTangent=-3.8261, tangentMode=RCTM_User, time=0.7, value=0}
          ]}}}
       Scale Curve=0.9
      )
      Z=&FloatFromCurve(FloatCurve=DataInterface() {propertyValues {Curve {keys [
           {interpMode=RCIM_Cubic, tangentMode=RCTM_User, time=0, value=1}
           {arriveTangent=-2.3974, interpMode=RCIM_Cubic, leaveTangent=-2.3974, tangentMode=RCTM_User, time=0.7, value=0}
          ]}}})
     )
    )) {isSetParameters=true}
   &ScaleColor(
    Scale Alpha=&Lerp_Float(
     A=4
     B=0.15
     Alpha=&FloatFromCurve(FloatCurve=DataInterface() {propertyValues {Curve {keys [{interpMode=RCIM_Cubic, tangentMode=RCTM_User, time=0, value=0}, {interpMode=RCIM_Cubic, tangentMode=RCTM_User, time=0.7, value=1}]}}})
    )
    ScaleA=true
    ScaleRGB=true
   )
   /Game/Effects/Niagara/Modules/Materials/FN_MeshMaterial(
    Amplitude=2
    Falloff=-15
    Frequency=20
    Jitter Multiplier=3
    Jitter Pan by SpawnGroups=false
    Offset=10
    Scale Noise X Axis=2
    Scale Noise Y Axis=2
    Speed=1
   )
  ]
  rendererClasses [&NiagaraMeshRendererProperties]
  renderers [
   {rendererClass=&NiagaraMeshRendererProperties, rendererIndex=0}
   {rendererClass=&NiagaraMeshRendererProperties, rendererIndex=1}
   {rendererClass=&NiagaraMeshRendererProperties, rendererIndex=2}
  ]
 }
 Liquid_Particles {
  emitterUpdateScript [
   &EmitterState()
   &SpawnParticlesFromOtherEmitter(
    Attribute Reader=DataInterface() {propertyValues {EmitterBinding {bindingMode=Other, emitterName=Cups}}}
    Minimum Age Spawn Threshold=0.5
    Spawn Rate=20
    Spawn Rate Per Particle Cap=150
    Use Maximum Age=false
    Use Minimum Age=false
   )
  ]
  particleSpawnScript [
   &InitializeParticle(
    Color Mode=::NewEnumerator0
    Lifetime Max=1
    Lifetime Min=0.75
    Lifetime Mode=::NewEnumerator1
    Sprite Rotation Mode=::NewEnumerator2
    Sprite Size Max=<20 40>
    Sprite Size Min=<15 30>
    Sprite Size Mode=::NewEnumerator2
    UsePositionOffset=false
    Write Material Random=true
   )
   &SampleParticlesFromOtherEmitter(
    Color Sampling=::NewEnumerator1
    Maximum Sampled Particle Age=0.9
    Mesh Orientation Sampling:ENiagara_DirectReadSamplingMode=::NewEnumerator1
    Minimum Sampled Particle Age=0.1
    Module.MaxSampledAgeCutoff=true
    Module.MinSampledAgeCutoff=true
    Particle ID Sampling=::NewEnumerator1
    Reject Particle By Sampled Age=true
   )
   /Niagara/Modules/Update/SubUV/SubUVAnimation(Number Of Frames=8, SubUV Animation Mode=::NewEnumerator1)
   &Color(Color=&MakeLinearColorFromVectorAndFloat(
     "Vector (RGB)"=&MakeVectorFromLinearColorRGB(Color=$Output.SampleParticlesFromOtherEmitter.SampledColor)
     "Float (Alpha)"=1
    ))
   /SetVariables(Particles.TrailDirection:Vector3f=&Add_Vector(
     A=$System.BobaVelocity
     B=&Multiply_VectorByFloat(
      Vector=&ReturnMeshOrientationAxis(Mesh Orientation Axis=::ZAxis, Mesh Orientation Quat=$Output.SampleParticlesFromOtherEmitter.SampledMeshOrientation)
      Float=$System.SwirlSpread
     )
    )) {enabled=false, isSetParameters=true}
   /SetVariables(Particles.TrailDirection:Vector3f=&Add_Vector(
     A=$System.BobaVelocity
     B=&MakeVector(
      X=&Sine(Normalized Angle=$Emitter.Age, Period=0.66, Scale=$System.SwirlSpread, Bias=0)
      Y=&Cosine(Normalized Angle=$Emitter.Age, Period=0.66, Scale=$System.SwirlSpread, Bias=0)
      Z=0
     )
    )) {enabled=false, isSetParameters=true}
   /SetVariables(Particles.TrailDirection:Vector3f=&MakeVector(
     X=&Sine(Normalized Angle=$Emitter.Age, Period=0.66, Scale=$System.SwirlSpread, Bias=0)
     Y=&Cosine(Normalized Angle=$Emitter.Age, Period=0.66, Scale=$System.SwirlSpread, Bias=0)
     Z=0
    )) {isSetParameters=true}
  ]
  particleUpdateScript [
   &ParticleState()
   /SetVariables(Particles.PositionOffset:NiagaraPosition=&Add_Vector(
     A=Linked() {linkedVariable {name=Particles.Position, type=NiagaraPosition}}
     B=&Multiply_VectorByFloat(
      Vector=&Add_Vector(A=$System.BobaVelocity, B=$Particles.TrailDirection)
      Float=&FN_FrontEndFloatModifier(
       Gameplay=&FloatFromCurve(
        FloatCurve=DataInterface() {propertyValues {Curve {keys [
            {arriveTangent=0.5537, interpMode=RCIM_Cubic, leaveTangent=0.5537, tangentMode=RCTM_User, time=0, value=0}
            {arriveTangent=1.1281, interpMode=RCIM_Cubic, leaveTangent=1.1281, tangentMode=RCTM_User, time=1, value=1}
           ]}}}
        Scale Curve=&Multiply_Float(A=$System.BobaDistance, B=$Particles.Lifetime)
       )
       Front End=&FloatFromCurve(
        FloatCurve=DataInterface() {propertyValues {Curve {keys [
            {arriveTangent=0.5537, interpMode=RCIM_Cubic, leaveTangent=0.5537, tangentMode=RCTM_User, time=0, value=0}
            {arriveTangent=1.1281, interpMode=RCIM_Cubic, leaveTangent=1.1281, tangentMode=RCTM_User, time=1, value=1}
           ]}}}
        Scale Curve=&Multiply_Float(A=$System.BobaDistance, B=$Particles.Lifetime)
       )
      )
     )
    )) {isSetParameters=true}
   &SpriteRotationRate(Rotation Rate=&RandomRangeFloat(Minimum=-20, Maximum=20))
   &ScaleSpriteSize(
    Scale Factor=&Multiply_Vector2D_ByFloat(
     Float=&FloatFromCurve(
      FloatCurve=DataInterface() {propertyValues {Curve {keys [
          {arriveTangent=4.1785, interpMode=RCIM_Cubic, leaveTangent=4.1785, tangentMode=RCTM_User, time=0, value=0.2}
          {arriveTangent=0.8, interpMode=RCIM_Cubic, leaveTangent=0.8, time=0.2042, value=0.7}
          {arriveTangent=0.0224, interpMode=RCIM_Cubic, leaveTangent=0.0224, tangentMode=RCTM_User, time=1, value=1}
         ]}}}
      Scale Curve=1.5
     )
     Vector2D=&Lerp_Vector2(
      A=<1 0.5>
      B=<1 1>
      Alpha=&FloatFromCurve(FloatCurve=DataInterface() {propertyValues {Curve {keys [
           {arriveTangent=4.1785, interpMode=RCIM_Cubic, leaveTangent=4.1785, tangentMode=RCTM_User, time=0, value=0}
           {arriveTangent=1.1179, interpMode=RCIM_Cubic, leaveTangent=1.1179, tangentMode=RCTM_User, time=0.2042, value=0.6228}
           {arriveTangent=0.0224, interpMode=RCIM_Cubic, leaveTangent=0.0224, tangentMode=RCTM_User, time=1, value=1}
          ]}}})
     )
    )
    Scale Sprite Size Mode=::NewEnumerator2
   )
   &ScaleColor(
    Scale Alpha=&FloatFromCurve(
     FloatCurve=DataInterface() {propertyValues {Curve {keys [{interpMode=RCIM_Cubic, tangentMode=RCTM_User, time=0.66, value=1}, {interpMode=RCIM_Cubic, tangentMode=RCTM_User, time=1, value=1}]}}}
     Scale Curve=0.75
    )
    Scale RGB=&FN_FrontEndVector3Modifier(Gameplay=<2 2 2>, Front End=<2 2 2>)
    ScaleA=false
    ScaleRGB=true
   )
   &DynamicMaterialParameters(
    Index 0 Param 1=&RandomRangeFloat(Minimum=0, Maximum=1)
    Index 0 Param 2=&FloatFromCurve(
     FloatCurve=DataInterface() {propertyValues {Curve {keys [{interpMode=RCIM_Cubic, tangentMode=RCTM_User, time=0.33, value=0}, {interpMode=RCIM_Cubic, leaveTangent=1.2673, time=0.8, value=1}]}}}
     Scale Curve=0
    )
    Index 0 Param 3=&FloatFromCurve(FloatCurve=DataInterface() {propertyValues {Curve {keys [{tangentMode=RCTM_User, time=0, value=1}, {tangentMode=RCTM_User, time=1, value=1}]}}})
    Index 0 Param 4=&FloatFromCurve(FloatCurve=DataInterface() {propertyValues {Curve {keys [
         {arriveTangent=-14.9399, interpMode=RCIM_Cubic, leaveTangent=-14.9399, tangentMode=RCTM_User, time=0, value=1}
         {interpMode=RCIM_Cubic, tangentMode=RCTM_User, time=0.15, value=0}
         {interpMode=RCIM_Cubic, tangentMode=RCTM_User, time=0.33, value=0}
         {arriveTangent=2.7964, interpMode=RCIM_Cubic, leaveTangent=2.7964, tangentMode=RCTM_User, time=1, value=1}
        ]}}})
    Index 1 Param 2=&FloatFromCurve(FloatCurve=DataInterface() {propertyValues {Curve {keys [
         {arriveTangent=23.2225, interpMode=RCIM_Cubic, leaveTangent=23.2225, tangentMode=RCTM_User, time=0, value=0}
         <at 0.1 1 RCIM_Cubic>
         {arriveTangent=-3.2993, interpMode=RCIM_Cubic, leaveTangent=-3.2993, tangentMode=RCTM_User, time=1, value=0}
        ]}}})
    Module.Param0WriteEnabled=true
    Module.Param0WriteEnabled1=true
    Module.Param1WriteEnabled=true
    Module.Param1WriteEnabled1=true
    Module.Param2WriteEnabled=true
    Module.Param2WriteEnabled1=false
    Module.Param3WriteEnabled=true
    Module.Param3WriteEnabled1=false
    Write Parameter Index 1=true
   )
  ]
  rendererClasses [&NiagaraSpriteRendererProperties]
  renderers [{rendererClass=&NiagaraSpriteRendererProperties}]
 }
 Liquid_Trail {
  bEnabled=false
  emitterUpdateScript [
   &EmitterState()
   &SpawnParticlesFromOtherEmitter(
    Attribute Reader=DataInterface() {propertyValues {EmitterBinding {bindingMode=Other, emitterName=Cups}}}
    Minimum Age Spawn Threshold=0.5
    Spawn Rate=20
    Spawn Rate Per Particle Cap=150
    Use Maximum Age=false
    Use Minimum Age=false
   )
  ]
  particleSpawnScript [
   &InitializeParticle(
    Color Mode=::NewEnumerator0
    Lifetime=0.8
    Mesh Scale Mode=::NewEnumerator1
    Mesh Uniform Scale Max=0.3
    Mesh Uniform Scale Min=0.25
    Position Mode=::NewEnumerator0
    Ribbon Width=35
    Ribbon Width Mode=::NewEnumerator1
    Sprite Rotation Angle=&RandomRangeFloat(Minimum=-0.05, Maximum=0.05)
    Sprite Rotation Mode=::NewEnumerator1
    Sprite Size Mode=::NewEnumerator3
    Uniform Sprite Size=5
    Write Material Random=true
   )
   &SampleParticlesFromOtherEmitter(
    Age Sampling:ENiagara_DirectReadSamplingMode=::NewEnumerator1
    Color Sampling=::NewEnumerator1
    Maximum Sampled Particle Age=0.8
    Mesh Orientation Sampling:ENiagara_DirectReadSamplingMode=::NewEnumerator1
    Minimum Sampled Particle Age=0.1
    Module.MaxSampledAgeCutoff=true
    Module.MinSampledAgeCutoff=true
    Particle ID Sampling=::NewEnumerator1
    Reject Particle By Sampled Age=true
    Velocity Application:ENiagara_DirectReadApplicationMode=::NewEnumerator0
    Velocity Sampling=::NewEnumerator0
   )
   &Color(Color=&MakeLinearColorFromVectorAndFloat(
     "Vector (RGB)"=&MakeVectorFromLinearColorRGB(Color=$Output.SampleParticlesFromOtherEmitter.SampledColor)
     "Float (Alpha)"=0.75
    ))
   /SetVariables(Particles.TrailDirection:Vector3f=&Add_Vector(
     A=$System.BobaVelocity
     B=&Multiply_VectorByFloat(
      Vector=&ReturnMeshOrientationAxis(Mesh Orientation Axis=::ZAxis, Mesh Orientation Quat=$Output.SampleParticlesFromOtherEmitter.SampledMeshOrientation)
      Float=$System.SwirlSpread
     )
    )) {enabled=false, isSetParameters=true, module=SetVariables_#6}
   /SetVariables(Particles.TrailDirection:Vector3f=&Add_Vector(
     A=$System.BobaVelocity
     B=&MakeVector(
      X=&Sine(Normalized Angle=$Emitter.Age, Period=0.66, Scale=$System.SwirlSpread, Bias=0)
      Y=&Cosine(Normalized Angle=$Emitter.Age, Period=0.66, Scale=$System.SwirlSpread, Bias=0)
      Z=0
     )
    )) {enabled=false, isSetParameters=true, module=SetVariables_#7}
   /SetVariables(Particles.TrailDirection:Vector3f=&MakeVector(
     X=&Sine(Normalized Angle=$Emitter.Age, Period=0.66, Scale=$System.SwirlSpread, Bias=0)
     Y=&Cosine(Normalized Angle=$Emitter.Age, Period=0.66, Scale=$System.SwirlSpread, Bias=0)
     Z=0
    )) {isSetParameters=true, module=SetVariables_#8}
  ]
  particleUpdateScript [
   &ParticleState()
   /SetVariables(Particles.PositionOffset:NiagaraPosition=&Add_Vector(
     A=Linked() {linkedVariable {name=Particles.Position, type=NiagaraPosition}}
     B=&Multiply_VectorByFloat(
      Vector=&Add_Vector(A=$System.BobaVelocity, B=$Particles.TrailDirection)
      Float=&FN_FrontEndFloatModifier(
       Gameplay=&FloatFromCurve(
        FloatCurve=DataInterface() {propertyValues {Curve {keys [
            {arriveTangent=0.5537, interpMode=RCIM_Cubic, leaveTangent=0.5537, tangentMode=RCTM_User, time=0, value=0}
            {arriveTangent=1.1281, interpMode=RCIM_Cubic, leaveTangent=1.1281, tangentMode=RCTM_User, time=1, value=1}
           ]}}}
        Scale Curve=&Multiply_Float(A=$System.BobaDistance, B=$Particles.Lifetime)
       )
       Front End=&FloatFromCurve(
        FloatCurve=DataInterface() {propertyValues {Curve {keys [
            {arriveTangent=0.5537, interpMode=RCIM_Cubic, leaveTangent=0.5537, tangentMode=RCTM_User, time=0, value=0}
            {arriveTangent=1.1281, interpMode=RCIM_Cubic, leaveTangent=1.1281, tangentMode=RCTM_User, time=1, value=1}
           ]}}}
        Scale Curve=&Multiply_Float(A=$System.BobaDistance, B=$Particles.Lifetime)
       )
      )
     )
    )) {isSetParameters=true, module=SetVariables_#9}
   /SetVariables(Particles.RibbonUVDistance:NiagaraFloat=$Particles.NormalizedAge) {isSetParameters=true, module=SetVariables_#10}
   &ScaleSpriteSize(
    Scale Factor=&Multiply_Vector2D_ByFloat(
     Float=&FloatFromCurve(
      FloatCurve=DataInterface() {propertyValues {Curve {keys [
          {arriveTangent=4.1785, interpMode=RCIM_Cubic, leaveTangent=4.1785, tangentMode=RCTM_User, time=0, value=0.5}
          {arriveTangent=0.3888, interpMode=RCIM_Cubic, leaveTangent=0.3888, tangentMode=RCTM_User, time=0.25, value=0.9}
          {arriveTangent=0.1345, interpMode=RCIM_Cubic, leaveTangent=0.1345, tangentMode=RCTM_User, time=1, value=1}
         ]}}}
      Scale Curve=1.25
     )
     Vector2D=&Lerp_Vector2(
      A=<1 1>
      B=<1 1>
      Alpha=&FloatFromCurve(FloatCurve=DataInterface() {propertyValues {Curve {keys [
           {arriveTangent=4.1785, interpMode=RCIM_Cubic, leaveTangent=4.1785, tangentMode=RCTM_User, time=0, value=0}
           {arriveTangent=1.1179, interpMode=RCIM_Cubic, leaveTangent=1.1179, tangentMode=RCTM_User, time=0.2042, value=0.6228}
           {arriveTangent=0.0224, interpMode=RCIM_Cubic, leaveTangent=0.0224, tangentMode=RCTM_User, time=1, value=1}
          ]}}})
     )
    )
    Scale Sprite Size Mode=::NewEnumerator2
   )
   &ScaleRibbonWidth(Ribbon Width Scale=&FloatFromCurve(
     FloatCurve=DataInterface() {propertyValues {Curve {keys [
         {interpMode=RCIM_Cubic, tangentMode=RCTM_User, time=0, value=0.75}
         {interpMode=RCIM_Cubic, tangentMode=RCTM_User, time=0.5, value=1}
         {tangentMode=RCTM_User, time=1, value=0.5}
        ]}}}
     CurveIndex=&ReturnNormalizedExecIndex()
    ))
   &ScaleColor(
    Scale RGB=&FN_FrontEndVector3Modifier(Gameplay=<1 1 1>, Front End=<2 2 2>)
    ScaleA=false
    ScaleRGB=true
   )
   &DynamicMaterialParameters(
    Index 0 Param 1=&Multiply_FloatByInt(
     Float=0.5
     Integer=&ReturnParticleIDAttributes(Particle ID Attribute=::NewEnumerator1, Niagara ID=$Particles.RibbonID)
    )
    Index 0 Param 2=&FloatFromCurve(
     FloatCurve=DataInterface() {propertyValues {Curve {keys [
         {arriveTangent=12.9462, interpMode=RCIM_Cubic, time=0, value=0}
         {interpMode=RCIM_Cubic, tangentMode=RCTM_User, time=0.2, value=1}
         {interpMode=RCIM_Cubic, tangentMode=RCTM_User, time=0.8, value=1}
         {interpMode=RCIM_Cubic, leaveTangent=-5.277, time=1, value=0}
        ]}}}
     CurveIndex=&ReturnNormalizedExecIndex()
     Scale Curve=&FloatFromCurve(FloatCurve=DataInterface() {propertyValues {Curve {keys [{interpMode=RCIM_Cubic, tangentMode=RCTM_User, time=0.5, value=1}, {interpMode=RCIM_Cubic, tangentMode=RCTM_User, time=1, value=0}]}}})
    )
    Module.Param0WriteEnabled=true
    Module.Param1WriteEnabled=true
    Module.Param2WriteEnabled=true
    Module.Param3WriteEnabled=true
   )
  ]
  rendererClasses [&NiagaraRibbonRendererProperties, &NiagaraSpriteRendererProperties]
  renderers [
   {rendererClass=&NiagaraRibbonRendererProperties}
   {rendererClass=&NiagaraSpriteRendererProperties, rendererIndex=1}
  ]
 }
 Liquid_Trail002 {
  emitterUpdateScript [
   &EmitterState()
   &SpawnParticlesFromOtherEmitter(
    Attribute Reader=DataInterface() {propertyValues {EmitterBinding {bindingMode=Other, emitterName=Cups}}}
    Minimum Age Spawn Threshold=0.5
    Spawn Rate=20
    Spawn Rate Per Particle Cap=150
    Use Maximum Age=false
    Use Minimum Age=false
   )
  ]
  particleSpawnScript [
   &InitializeParticle(
    Color Mode=::NewEnumerator0
    Lifetime=0.8
    Mesh Scale Mode=::NewEnumerator1
    Mesh Uniform Scale Max=0.3
    Mesh Uniform Scale Min=0.25
    Position Mode=::NewEnumerator0
    Ribbon Width=35
    Ribbon Width Mode=::NewEnumerator1
    Sprite Rotation Angle=&RandomRangeFloat(Minimum=-0.05, Maximum=0.05)
    Sprite Rotation Mode=::NewEnumerator1
    Sprite Size Mode=::NewEnumerator3
    Uniform Sprite Size=5
    Write Material Random=true
   )
   &SampleParticlesFromOtherEmitter(
    Age Sampling:ENiagara_DirectReadSamplingMode=::NewEnumerator1
    Color Sampling=::NewEnumerator1
    Maximum Sampled Particle Age=0.8
    Mesh Orientation Sampling:ENiagara_DirectReadSamplingMode=::NewEnumerator1
    Minimum Sampled Particle Age=0.1
    Module.MaxSampledAgeCutoff=true
    Module.MinSampledAgeCutoff=true
    Particle ID Sampling=::NewEnumerator1
    Reject Particle By Sampled Age=true
    Velocity Application:ENiagara_DirectReadApplicationMode=::NewEnumerator0
    Velocity Sampling=::NewEnumerator0
   )
   /SetVariables(Particles.NormalizedSampledAge:NiagaraFloat=&RemapRange(Input Value To Remap=$Output.SampleParticlesFromOtherEmitter.SampledAge, Input Min=0.1167, Input Max=0.7833, Output Min=0, Output Max=1)) {isSetParameters=true}
   &Color(Color=&MakeLinearColorFromVectorAndFloat(
     "Vector (RGB)"=&MakeVectorFromLinearColorRGB(Color=$Output.SampleParticlesFromOtherEmitter.SampledColor)
     "Float (Alpha)"=1
    ))
   /SetVariables(Particles.TrailDirection:Vector3f=&Add_Vector(
     A=$System.BobaVelocity
     B=&Multiply_VectorByFloat(
      Vector=&ReturnMeshOrientationAxis(Mesh Orientation Axis=::ZAxis, Mesh Orientation Quat=$Output.SampleParticlesFromOtherEmitter.SampledMeshOrientation)
      Float=$System.SwirlSpread
     )
    )) {enabled=false, isSetParameters=true, module=SetVariables_#6}
   /SetVariables(Particles.TrailDirection:Vector3f=&Add_Vector(
     A=$System.BobaVelocity
     B=&MakeVector(
      X=&Sine(Normalized Angle=$Emitter.Age, Period=0.66, Scale=$System.SwirlSpread, Bias=0)
      Y=&Cosine(Normalized Angle=$Emitter.Age, Period=0.66, Scale=$System.SwirlSpread, Bias=0)
      Z=0
     )
    )) {enabled=false, isSetParameters=true, module=SetVariables_#7}
   /SetVariables(Particles.TrailDirection:Vector3f=&MakeVector(
     X=&Sine(Normalized Angle=$Emitter.Age, Period=0.66, Scale=$System.SwirlSpread, Bias=0)
     Y=&Cosine(Normalized Angle=$Emitter.Age, Period=0.66, Scale=$System.SwirlSpread, Bias=0)
     Z=0
    )) {isSetParameters=true, module=SetVariables_#8}
   /SetVariables(Particles.Random=&RandomRangeFloat(
     Minimum=0
     Maximum=1
     Randomness Mode=::NewEnumerator1
     Random Seed=&ReturnParticleIDAttributes(Particle ID Attribute=::NewEnumerator1, Niagara ID=$Particles.RibbonID)
     Fixed Random Seed=true
    )) {isSetParameters=true}
   /SetVariables(Particles.RibbonUVDistance:NiagaraFloat=&Multiply_Float(A=&Add_Float(A=$Output.SampleParticlesFromOtherEmitter.SampledAge, B=$Particles.Random), B=120)) {isSetParameters=true, module=SetVariables_#10}
   /SetVariables(Particles.RibbonLinkOrder:NiagaraFloat=$Particles.NormalizedSampledAge) {isSetParameters=true}
   /SetVariables(Particles.Velocity=<0 0 200>) {isSetParameters=true}
  ]
  particleUpdateScript [
   &ParticleState()
   /SetVariables(Particles.PositionOffset:NiagaraPosition=&Add_Vector(
     A=Linked() {linkedVariable {name=Particles.Position, type=NiagaraPosition}}
     B=&Multiply_VectorByFloat(
      Vector=&Add_Vector(A=$System.BobaVelocity, B=$Particles.TrailDirection)
      Float=&FN_FrontEndFloatModifier(
       Gameplay=&FloatFromCurve(
        FloatCurve=DataInterface() {propertyValues {Curve {keys [
            {arriveTangent=0.5537, interpMode=RCIM_Cubic, leaveTangent=0.5537, tangentMode=RCTM_User, time=0, value=0}
            {arriveTangent=1.1281, interpMode=RCIM_Cubic, leaveTangent=1.1281, tangentMode=RCTM_User, time=1, value=1}
           ]}}}
        Scale Curve=&Multiply_Float(A=$System.BobaDistance, B=$Particles.Lifetime)
       )
       Front End=&FloatFromCurve(
        FloatCurve=DataInterface() {propertyValues {Curve {keys [
            {arriveTangent=0.5537, interpMode=RCIM_Cubic, leaveTangent=0.5537, tangentMode=RCTM_User, time=0, value=0}
            {arriveTangent=1.1281, interpMode=RCIM_Cubic, leaveTangent=1.1281, tangentMode=RCTM_User, time=1, value=1}
           ]}}}
        Scale Curve=&Multiply_Float(A=$System.BobaDistance, B=$Particles.Lifetime)
       )
      )
     )
    )) {isSetParameters=true, module=SetVariables_#9}
   &ScaleSpriteSize(
    Scale Factor=&Multiply_Vector2D_ByFloat(
     Float=&FloatFromCurve(
      FloatCurve=DataInterface() {propertyValues {Curve {keys [
          {arriveTangent=4.1785, interpMode=RCIM_Cubic, leaveTangent=4.1785, tangentMode=RCTM_User, time=0, value=0.5}
          {arriveTangent=0.3888, interpMode=RCIM_Cubic, leaveTangent=0.3888, tangentMode=RCTM_User, time=0.25, value=0.9}
          {arriveTangent=0.1345, interpMode=RCIM_Cubic, leaveTangent=0.1345, tangentMode=RCTM_User, time=1, value=1}
         ]}}}
      Scale Curve=1.25
     )
     Vector2D=&Lerp_Vector2(
      A=<1 1>
      B=<1 1>
      Alpha=&FloatFromCurve(FloatCurve=DataInterface() {propertyValues {Curve {keys [
           {arriveTangent=4.1785, interpMode=RCIM_Cubic, leaveTangent=4.1785, tangentMode=RCTM_User, time=0, value=0}
           {arriveTangent=1.1179, interpMode=RCIM_Cubic, leaveTangent=1.1179, tangentMode=RCTM_User, time=0.2042, value=0.6228}
           {arriveTangent=0.0224, interpMode=RCIM_Cubic, leaveTangent=0.0224, tangentMode=RCTM_User, time=1, value=1}
          ]}}})
     )
    )
    Scale Sprite Size Mode=::NewEnumerator2
   )
   &ScaleRibbonWidth(Ribbon Width Scale=&FloatFromCurve(
     FloatCurve=DataInterface() {propertyValues {Curve {keys [
         {interpMode=RCIM_Cubic, tangentMode=RCTM_User, time=0, value=0.75}
         {interpMode=RCIM_Cubic, tangentMode=RCTM_User, time=0.5, value=1}
         {tangentMode=RCTM_User, time=1, value=0.5}
        ]}}}
     CurveIndex=$Particles.NormalizedSampledAge
    ))
   &ScaleColor(
    Scale RGB=&FN_FrontEndVector3Modifier(Gameplay=<1 1 1>, Front End=<2 2 2>)
    ScaleA=false
    ScaleRGB=true
   ) {enabled=false}
   &ScaleColor(
    Scale Alpha=&FloatFromCurve(
     FloatCurve=DataInterface() {propertyValues {Curve {keys [{interpMode=RCIM_Cubic, tangentMode=RCTM_User, time=0.66, value=1}, {interpMode=RCIM_Cubic, tangentMode=RCTM_User, time=1, value=1}]}}}
     Scale Curve=0.75
    )
    Scale RGB=&FN_FrontEndVector3Modifier(Gameplay=<2 2 2>, Front End=<2 2 2>)
    ScaleA=false
    ScaleRGB=true
   ) {module=ScaleColor001}
   &DynamicMaterialParameters(
    Index 0 Param 1=$Particles.Random
    Index 0 Param 2=&FloatFromCurve(
     FloatCurve=DataInterface() {propertyValues {Curve {keys [{interpMode=RCIM_Cubic, tangentMode=RCTM_User, time=0.33, value=0}, {interpMode=RCIM_Cubic, leaveTangent=1.2673, time=0.8, value=1}]}}}
     Scale Curve=0
    )
    Index 0 Param 4=&OneMinusFloat(Float=&FloatFromCurve(
      FloatCurve=DataInterface() {propertyValues {Curve {keys [
          {arriveTangent=12.9462, interpMode=RCIM_Cubic, time=0, value=0}
          {interpMode=RCIM_Cubic, tangentMode=RCTM_User, time=0.2, value=1}
          {interpMode=RCIM_Cubic, tangentMode=RCTM_User, time=0.8, value=1}
          {interpMode=RCIM_Cubic, leaveTangent=-5.277, time=1, value=0}
         ]}}}
      CurveIndex=$Particles.NormalizedSampledAge
      Scale Curve=&FloatFromCurve(FloatCurve=DataInterface() {propertyValues {Curve {keys [
           <at 0 0 RCIM_Cubic>
           {interpMode=RCIM_Cubic, tangentMode=RCTM_User, time=0.1, value=1}
           {interpMode=RCIM_Cubic, tangentMode=RCTM_User, time=0.5, value=1}
           {interpMode=RCIM_Cubic, tangentMode=RCTM_User, time=1, value=0}
          ]}}})
     ))
    Module.Index0Param0Write=true
    Module.Index0Param1Write=true
    Module.Index0Param2Write=true
    Module.Index0Param3Write=true
   )
  ]
  rendererClasses [&NiagaraRibbonRendererProperties, &NiagaraSpriteRendererProperties]
  renderers [
   {rendererClass=&NiagaraRibbonRendererProperties}
   {rendererClass=&NiagaraSpriteRendererProperties, rendererIndex=1}
   {rendererClass=&NiagaraRibbonRendererProperties, rendererIndex=2}
  ]
 }
}
summary {
 systemName=NS_Contrail_SharpFang
 emitters [
  {emitterName=Cups, rendererClasses [&NiagaraMeshRendererProperties]}
  {
   bEnabled=false
   emitterName=Liquid_Trail
   rendererClasses [&NiagaraRibbonRendererProperties, &NiagaraSpriteRendererProperties]
  }
  {emitterName=Liquid_Particles, rendererClasses [&NiagaraSpriteRendererProperties]}
  {emitterName=Boba_and_Ice, rendererClasses [&NiagaraMeshRendererProperties, &NiagaraSpriteRendererProperties]}
  {
   emitterName=Liquid_Trail002
   rendererClasses [&NiagaraRibbonRendererProperties, &NiagaraSpriteRendererProperties]
  }
 ]
 userVariables [
  User.bIsFrontEnd {type=NiagaraBool}
  User.bIsFrontEndPreview {type=NiagaraBool}
  User.ColorVariant {type=NiagaraFloat, defaultValue=0, description=""}
  User.ParaGlideLeanAlpha {type=NiagaraFloat, defaultValue=0, description=""}
 ]
}
